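# Landscape test material for Athena terrain that weight-blends eight paint layers, named 1 to 8.
# One layer blend mixes eight ground and surface textures, then multiplies by a Brightness parameter.
# Parallel blends supply normals, alternating grass and pebble maps, and flat debug-style colors per layer.
# The eight surface textures use shared wrap samplers, which saves sampler slots.
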Material /FigmentEnvironment/LM_8_1_1/Athena/Environments/Landscape/Material/LayerTest/M_Athena_LayerTest
graph {
 properties {pixelDepthOffsetMode=PDOM_Legacy, usedShadingModels=Default Lit}
 nodes [
  MaterialExpressionTextureSample_6 = TextureSample() {samplerType=SAMPLERTYPE_Normal, texture=/FigmentEnvironment/LM_8_1_1/Packages/DS_Fortnight/T_FORT_Grass02_N}
  MaterialExpressionTextureSample_4 = TextureSample() {samplerType=SAMPLERTYPE_Normal, texture=/FigmentEnvironment/LM_8_1_1/Packages/DS_Fortnight/pebbles_N}
  LandscapeLayerBlend!(
   Layer 1=@MaterialExpressionTextureSample_6.RGB
   Layer 2=@MaterialExpressionTextureSample_4.RGB
   Layer 3=@MaterialExpressionTextureSample_6.RGB
   Layer 4=@MaterialExpressionTextureSample_4.RGB
   Layer 5=@MaterialExpressionTextureSample_6.RGB
   Layer 6=@MaterialExpressionTextureSample_4.RGB
   Layer 7=@MaterialExpressionTextureSample_6.RGB
   Layer 8=@MaterialExpressionTextureSample_4.RGB
  ) {layers [
    {
     blendType=LB_WeightBlend
     constHeightInput=0
     layerName=1
     previewWeight=1
     constLayerInput=<0 0 0>
     heightInput {expression=null, inputName=None, mask=0, maskA=0, maskB=0, maskG=0, maskR=0, outputIndex=0}
     layerInput {
      expression=&self:MaterialExpressionTextureSample_6
      inputName=None
      mask=1
      maskA=0
      maskB=1
      maskG=1
      maskR=1
      outputIndex=0
     }
    }
    {
     blendType=LB_WeightBlend
     constHeightInput=0
     layerName=2
     previewWeight=1
     constLayerInput=<0 0 0>
     heightInput {expression=null, inputName=None, mask=0, maskA=0, maskB=0, maskG=0, maskR=0, outputIndex=0}
     layerInput {
      expression=&self:MaterialExpressionTextureSample_4
      inputName=None
      mask=1
      maskA=0
      maskB=1
      maskG=1
      maskR=1
      outputIndex=0
     }
    }
    {
     blendType=LB_WeightBlend
     constHeightInput=0
     layerName=3
     previewWeight=1
     constLayerInput=<0 0 0>
     heightInput {expression=null, inputName=None, mask=0, maskA=0, maskB=0, maskG=0, maskR=0, outputIndex=0}
     layerInput {
      expression=&self:MaterialExpressionTextureSample_6
      inputName=None
      mask=1
      maskA=0
      maskB=1
      maskG=1
      maskR=1
      outputIndex=0
     }
    }
    {
     blendType=LB_WeightBlend
     constHeightInput=0
     layerName=4
     previewWeight=1
     constLayerInput=<0 0 0>
     heightInput {expression=null, inputName=None, mask=0, maskA=0, maskB=0, maskG=0, maskR=0, outputIndex=0}
     layerInput {
      expression=&self:MaterialExpressionTextureSample_4
      inputName=None
      mask=1
      maskA=0
      maskB=1
      maskG=1
      maskR=1
      outputIndex=0
     }
    }
    {
     blendType=LB_WeightBlend
     constHeightInput=0
     layerName=5
     previewWeight=1
     constLayerInput=<0 0 0>
     heightInput {expression=null, inputName=None, mask=0, maskA=0, maskB=0, maskG=0, maskR=0, outputIndex=0}
     layerInput {
      expression=&self:MaterialExpressionTextureSample_6
      inputName=None
      mask=1
      maskA=0
      maskB=1
      maskG=1
      maskR=1
      outputIndex=0
     }
    }
    {
     blendType=LB_WeightBlend
     constHeightInput=0
     layerName=6
     previewWeight=1
     constLayerInput=<0 0 0>
     heightInput {expression=null, inputName=None, mask=0, maskA=0, maskB=0, maskG=0, maskR=0, outputIndex=0}
     layerInput {
      expression=&self:MaterialExpressionTextureSample_4
      inputName=None
      mask=1
      maskA=0
      maskB=1
      maskG=1
      maskR=1
      outputIndex=0
     }
    }
    {
     blendType=LB_WeightBlend
     constHeightInput=0
     layerName=7
     previewWeight=0
     constLayerInput=<0 0 0>
     heightInput {expression=null, inputName=None, mask=0, maskA=0, maskB=0, maskG=0, maskR=0, outputIndex=0}
     layerInput {
      expression=&self:MaterialExpressionTextureSample_6
      inputName=None
      mask=1
      maskA=0
      maskB=1
      maskG=1
      maskR=1
      outputIndex=0
     }
    }
    {
     blendType=LB_WeightBlend
     constHeightInput=0
     layerName=8
     previewWeight=0
     constLayerInput=<0 0 0>
     heightInput {expression=null, inputName=None, mask=0, maskA=0, maskB=0, maskG=0, maskR=0, outputIndex=0}
     layerInput {
      expression=&self:MaterialExpressionTextureSample_4
      inputName=None
      mask=1
      maskA=0
      maskB=1
      maskG=1
      maskR=1
      outputIndex=0
     }
    }
   ]}
  MaterialExpressionConstant3Vector_2 = Constant3Vector() {constant=<rgba 1 0 0 1>}
  MaterialExpressionConstant3Vector_5 = Constant3Vector() {constant=<rgba 1 0.4898 0 1>}
  MaterialExpressionConstant3Vector_8 = Constant3Vector() {constant=<rgba 0.9898 1 0 1>}
  MaterialExpressionConstant3Vector_11 = Constant3Vector() {constant=<rgba 0.1331 1 0 1>}
  MaterialExpressionConstant3Vector_14 = Constant3Vector() {constant=<rgba 0 0.8691 1 1>}
  MaterialExpressionConstant3Vector_17 = Constant3Vector() {constant=<rgba 0.8159 0 1 1>}
  MaterialExpressionConstant3Vector_20 = Constant3Vector() {constant=<rgba 0 0.1963 1 1>}
  MaterialExpressionConstant3Vector_21 = Constant3Vector() {constant=<rgba 0.1366 0 1 1>}
  LandscapeLayerBlend(
   Layer 1=@MaterialExpressionConstant3Vector_2
   Layer 2=@MaterialExpressionConstant3Vector_5
   Layer 3=@MaterialExpressionConstant3Vector_8
   Layer 4=@MaterialExpressionConstant3Vector_11
   Layer 5=@MaterialExpressionConstant3Vector_14
   Layer 6=@MaterialExpressionConstant3Vector_17
   Layer 7=@MaterialExpressionConstant3Vector_20
   Layer 8=@MaterialExpressionConstant3Vector_21
  ) {layers [
    {
     blendType=LB_WeightBlend
     constHeightInput=0
     layerName=1
     previewWeight=1
     constLayerInput=<0 0 0>
     heightInput {expression=null, inputName=None, mask=0, maskA=0, maskB=0, maskG=0, maskR=0, outputIndex=0}
     layerInput {
      expression=&self:MaterialExpressionConstant3Vector_2
      inputName=None
      mask=0
      maskA=0
      maskB=0
      maskG=0
      maskR=0
      outputIndex=0
     }
    }
    {
     blendType=LB_WeightBlend
     constHeightInput=0
     layerName=2
     previewWeight=1
     constLayerInput=<0 0 0>
     heightInput {expression=null, inputName=None, mask=0, maskA=0, maskB=0, maskG=0, maskR=0, outputIndex=0}
     layerInput {
      expression=&self:MaterialExpressionConstant3Vector_5
      inputName=None
      mask=0
      maskA=0
      maskB=0
      maskG=0
      maskR=0
      outputIndex=0
     }
    }
    {
     blendType=LB_WeightBlend
     constHeightInput=0
     layerName=3
     previewWeight=1
     constLayerInput=<0 0 0>
     heightInput {expression=null, inputName=None, mask=0, maskA=0, maskB=0, maskG=0, maskR=0, outputIndex=0}
     layerInput {
      expression=&self:MaterialExpressionConstant3Vector_8
      inputName=None
      mask=0
      maskA=0
      maskB=0
      maskG=0
      maskR=0
      outputIndex=0
     }
    }
    {
     blendType=LB_WeightBlend
     constHeightInput=0
     layerName=4
     previewWeight=1
     constLayerInput=<0 0 0>
     heightInput {expression=null, inputName=None, mask=0, maskA=0, maskB=0, maskG=0, maskR=0, outputIndex=0}
     layerInput {
      expression=&self:MaterialExpressionConstant3Vector_11
      inputName=None
      mask=0
      maskA=0
      maskB=0
      maskG=0
      maskR=0
      outputIndex=0
     }
    }
    {
     blendType=LB_WeightBlend
     constHeightInput=0
     layerName=5
     previewWeight=1
     constLayerInput=<0 0 0>
     heightInput {expression=null, inputName=None, mask=0, maskA=0, maskB=0, maskG=0, maskR=0, outputIndex=0}
     layerInput {
      expression=&self:MaterialExpressionConstant3Vector_14
      inputName=None
      mask=0
      maskA=0
      maskB=0
      maskG=0
      maskR=0
      outputIndex=0
     }
    }
    {
     blendType=LB_WeightBlend
     constHeightInput=0
     layerName=6
     previewWeight=1
     constLayerInput=<0 0 0>
     heightInput {expression=null, inputName=None, mask=0, maskA=0, maskB=0, maskG=0, maskR=0, outputIndex=0}
     layerInput {
      expression=&self:MaterialExpressionConstant3Vector_17
      inputName=None
      mask=0
      maskA=0
      maskB=0
      maskG=0
      maskR=0
      outputIndex=0
     }
    }
    {
     blendType=LB_WeightBlend
     constHeightInput=0
     layerName=7
     previewWeight=0
     constLayerInput=<0 0 0>
     heightInput {expression=null, inputName=None, mask=0, maskA=0, maskB=0, maskG=0, maskR=0, outputIndex=0}
     layerInput {
      expression=&self:MaterialExpressionConstant3Vector_20
      inputName=None
      mask=0
      maskA=0
      maskB=0
      maskG=0
      maskR=0
      outputIndex=0
     }
    }
    {
     blendType=LB_WeightBlend
     constHeightInput=0
     layerName=8
     previewWeight=0
     constLayerInput=<0 0 0>
     heightInput {expression=null, inputName=None, mask=0, maskA=0, maskB=0, maskG=0, maskR=0, outputIndex=0}
     layerInput {
      expression=&self:MaterialExpressionConstant3Vector_21
      inputName=None
      mask=0
      maskA=0
      maskB=0
      maskG=0
      maskR=0
      outputIndex=0
     }
    }
   ]}
  MaterialExpressionTextureSample_7 = TextureSample() {samplerSource=SSM_Wrap_WorldGroupSettings, texture=/FigmentEnvironment/LM_8_1_1/Packages/DS_Fortnight/T_FORT_Grass02}
  MaterialExpressionTextureSample_8 = TextureSample() {samplerSource=SSM_Wrap_WorldGroupSettings, texture=/FigmentEnvironment/LM_8_1_1/Packages/DS_Fortnight/SM/Textures/T_FORT_ForestFloor01}
  MaterialExpressionTextureSample_9 = TextureSample() {samplerSource=SSM_Wrap_WorldGroupSettings, texture=/FigmentEnvironment/LM_8_1_1/Packages/DS_Fortnight/SM/Textures/T_Cliff_Rock_1}
  MaterialExpressionTextureSample_10 = TextureSample() {samplerSource=SSM_Wrap_WorldGroupSettings, texture=/FigmentEnvironment/LM_8_1_1/Packages/DS_Fortnight/SM/Textures/Gravel_Sandy}
  MaterialExpressionTextureSample_11 = TextureSample() {samplerSource=SSM_Wrap_WorldGroupSettings, texture=/FigmentEnvironment/LM_8_1_1/Packages/DS_Fortnight/SM/Textures/T_ParkingDeckConcrete}
  MaterialExpressionTextureSample_12 = TextureSample() {samplerSource=SSM_Wrap_WorldGroupSettings, texture=/FigmentEnvironment/LM_8_1_1/Packages/DS_Fortnight/SM/Textures/T_soccer_net_01_D}
  MaterialExpressionTextureSample_13 = TextureSample() {samplerSource=SSM_Wrap_WorldGroupSettings, texture=/FigmentEnvironment/LM_8_1_1/Packages/DS_Fortnight/SM/Textures/T_FORT_SaltFlat}
  MaterialExpressionTextureSample_14 = TextureSample() {samplerSource=SSM_Wrap_WorldGroupSettings, texture=/FigmentEnvironment/LM_8_1_1/Packages/DS_Fortnight/SM/Textures/T_Cliff_Desert_Rock_2}
  n1 = LandscapeLayerBlend(
   Layer 1=@MaterialExpressionTextureSample_7.RGB
   Layer 2=@MaterialExpressionTextureSample_8.RGB
   Layer 3=@MaterialExpressionTextureSample_9.RGB
   Layer 4=@MaterialExpressionTextureSample_10.RGB
   Layer 5=@MaterialExpressionTextureSample_11.RGB
   Layer 6=@MaterialExpressionTextureSample_12.RGB
   Layer 7=@MaterialExpressionTextureSample_13.RGB
   Layer 8=@MaterialExpressionTextureSample_14.RGB
  ) {layers [
    {
     blendType=LB_WeightBlend
     constHeightInput=0
     layerName=1
     previewWeight=1
     constLayerInput=<0 0 0>
     heightInput {expression=null, inputName=None, mask=0, maskA=0, maskB=0, maskG=0, maskR=0, outputIndex=0}
     layerInput {
      expression=&self:MaterialExpressionTextureSample_7
      inputName=None
      mask=1
      maskA=0
      maskB=1
      maskG=1
      maskR=1
      outputIndex=0
     }
    }
    {
     blendType=LB_WeightBlend
     constHeightInput=0
     layerName=2
     previewWeight=1
     constLayerInput=<0 0 0>
     heightInput {expression=null, inputName=None, mask=0, maskA=0, maskB=0, maskG=0, maskR=0, outputIndex=0}
     layerInput {
      expression=&self:MaterialExpressionTextureSample_8
      inputName=None
      mask=1
      maskA=0
      maskB=1
      maskG=1
      maskR=1
      outputIndex=0
     }
    }
    {
     blendType=LB_WeightBlend
     constHeightInput=0
     layerName=3
     previewWeight=1
     constLayerInput=<0 0 0>
     heightInput {expression=null, inputName=None, mask=0, maskA=0, maskB=0, maskG=0, maskR=0, outputIndex=0}
     layerInput {
      expression=&self:MaterialExpressionTextureSample_9
      inputName=None
      mask=1
      maskA=0
      maskB=1
      maskG=1
      maskR=1
      outputIndex=0
     }
    }
    {
     blendType=LB_WeightBlend
     constHeightInput=0
     layerName=4
     previewWeight=1
     constLayerInput=<0 0 0>
     heightInput {expression=null, inputName=None, mask=0, maskA=0, maskB=0, maskG=0, maskR=0, outputIndex=0}
     layerInput {
      expression=&self:MaterialExpressionTextureSample_10
      inputName=None
      mask=1
      maskA=0
      maskB=1
      maskG=1
      maskR=1
      outputIndex=0
     }
    }
    {
     blendType=LB_WeightBlend
     constHeightInput=0
     layerName=5
     previewWeight=1
     constLayerInput=<0 0 0>
     heightInput {expression=null, inputName=None, mask=0, maskA=0, maskB=0, maskG=0, maskR=0, outputIndex=0}
     layerInput {
      expression=&self:MaterialExpressionTextureSample_11
      inputName=None
      mask=1
      maskA=0
      maskB=1
      maskG=1
      maskR=1
      outputIndex=0
     }
    }
    {
     blendType=LB_WeightBlend
     constHeightInput=0
     layerName=6
     previewWeight=1
     constLayerInput=<0 0 0>
     heightInput {expression=null, inputName=None, mask=0, maskA=0, maskB=0, maskG=0, maskR=0, outputIndex=0}
     layerInput {
      expression=&self:MaterialExpressionTextureSample_12
      inputName=None
      mask=1
      maskA=0
      maskB=1
      maskG=1
      maskR=1
      outputIndex=0
     }
    }
    {
     blendType=LB_WeightBlend
     constHeightInput=0
     layerName=7
     previewWeight=0
     constLayerInput=<0 0 0>
     heightInput {expression=null, inputName=None, mask=0, maskA=0, maskB=0, maskG=0, maskR=0, outputIndex=0}
     layerInput {
      expression=&self:MaterialExpressionTextureSample_13
      inputName=None
      mask=1
      maskA=0
      maskB=1
      maskG=1
      maskR=1
      outputIndex=0
     }
    }
    {
     blendType=LB_WeightBlend
     constHeightInput=0
     layerName=8
     previewWeight=0
     constLayerInput=<0 0 0>
     heightInput {expression=null, inputName=None, mask=0, maskA=0, maskB=0, maskG=0, maskR=0, outputIndex=0}
     layerInput {
      expression=&self:MaterialExpressionTextureSample_14
      inputName=None
      mask=1
      maskA=0
      maskB=1
      maskG=1
      maskR=1
      outputIndex=0
     }
    }
   ]}
  Multiply(A=@n1, B=ScalarParameter() {defaultValue=0.1835, parameterName=Brightness})
  Multiply(A=@MaterialExpressionTextureSample_7.RGB)
 ]
}
